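# Physics asset for the SherWolf pickaxe axe mesh, a two-handed melee cosmetic weapon.
# It defines one body on the Melee_TwoHanded_Handle bone, made of a single capsule roughly aligned along the handle.
# It has no constraints, so the weapon collides as one rigid piece.
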PhysicsAsset /BRCosmetics/Weapons/FORT_Melee/Pickaxe_SherWolf/Meshes/SherWolf_Axe_PhysicsAsset
bodies {Melee_TwoHanded_Handle [{shapeType=Capsule, length=60.1182, radius=3.4009, center=<0.0719 -0.0083 28.075>, rotation=<rot 0.0503 -90 -0.2269>}]}
constraints []
