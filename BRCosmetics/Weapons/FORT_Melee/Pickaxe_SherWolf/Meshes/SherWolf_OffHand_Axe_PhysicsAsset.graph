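# Physics asset for the off-hand axe mesh of the SherWolf pickaxe, a melee weapon cosmetic.
# A single body on the Melee_TwoHanded_Handle bone uses one long, thin capsule, offset and rotated from the bone origin, as its collision shape.
# It has no constraints, so the asset defines no jointed or secondary body parts.
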PhysicsAsset /BRCosmetics/Weapons/FORT_Melee/Pickaxe_SherWolf/Meshes/SherWolf_OffHand_Axe_PhysicsAsset
bodies {Melee_TwoHanded_Handle [{shapeType=Capsule, length=40.634, radius=2.4909, center=<-0.0016 -0.0037 24.0723>, rotation=<rot -0.0053 -90 -0.0023>}]}
constraints []
